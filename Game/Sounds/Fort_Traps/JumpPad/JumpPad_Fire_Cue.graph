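# Sound cue for the jump pad trap's fire sound.
# A Random node picks between two equally weighted fire waves, and its output passes through a Modulator at the root.
# The cue uses the shared Fort traps attenuation and the trap placement sound class, with a concurrency override that prevents new instances once the limit is reached.
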SoundCue /Game/Sounds/Fort_Traps/JumpPad/JumpPad_Fire_Cue
root=Modulator_12
nodes [
 {id=Modulator_12, children [Random_30]}
 {id=Random_30, children [WavePlayer_108, WavePlayer_109], properties {weights [1,1]}}
 {id=WavePlayer_108, properties {soundWaveAssetPtr=&/JumpPad_Fire_001}}
 {id=WavePlayer_109, properties {soundWaveAssetPtr=&/JumpPad_Fire_002}}
]
properties {
 attenuationSettings=&Fort_Traps_Att
 bOverrideConcurrency=true
 maxDistance=6000
 soundClassObject=/Game/Sounds/FortSoundClassesAndModes/Fort_Traps_Placement
 volumeMultiplier=1
 concurrencyOverrides {resolutionRule=PreventNew}
}
