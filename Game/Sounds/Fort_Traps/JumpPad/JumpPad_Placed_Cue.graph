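# Sound cue for the placement sound of a Jump Pad trap, routed to the trap placement sound class.
# A mixer layers two placed-sound waves at once, with the JumpPad_Placed_001 layer slightly quieter than JumpPad_Placed_002.
# Playback uses the shared trap attenuation settings and overrides concurrency to prevent new instances once the limit is reached.
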SoundCue /Game/Sounds/Fort_Traps/JumpPad/JumpPad_Placed_Cue
root=Mixer_16
nodes [
 {id=Mixer_16, children [WavePlayer_106, WavePlayer_107], properties {inputVolume [1,0.7]}}
 {id=WavePlayer_106, properties {soundWaveAssetPtr=&/JumpPad_Placed_002}}
 {id=WavePlayer_107, properties {soundWaveAssetPtr=&/JumpPad_Placed_001}}
]
properties {
 attenuationSettings=&Fort_Traps_Att
 bOverrideConcurrency=true
 maxDistance=6000
 soundClassObject=/Game/Sounds/FortSoundClassesAndModes/Fort_Traps_Placement
 volumeMultiplier=0.9
 concurrencyOverrides {resolutionRule=PreventNew}
}
