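# Sound cue for the jump pad trap's reload, reusing a Speed Trap ready sound.
# A short fixed delay feeds an Enveloper that randomly pitches the wave down and holds volume before fading it out by 1.5 seconds.
# It plays at half volume in the trap placement sound class, with overridden attenuation using listener focus.
# Overridden concurrency prevents new instances once the limit is reached.
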SoundCue /Game/Sounds/Fort_Traps/JumpPad/JumpPad_Reload_Cue
root=Delay_0
nodes [
 {id=Delay_0, children [Enveloper_0], properties {delayMax=0.05, delayMin=0.05}}
 {
  id=Enveloper_0
  children [WavePlayer_0]
  properties {
   pitchMax=0.9
   pitchMin=0.8
   pitchCurve {editorCurveData {keys [{time=0}]}}
   volumeCurve {editorCurveData {keys [<at 0 1>, {arriveTangent=-0.8, interpMode=RCIM_Cubic, leaveTangent=-0.2045, tangentMode=RCTM_Break, time=0.6605, value=1}, <at 1.5 0>]}}
  }
 }
 {id=WavePlayer_0, properties {soundWaveAssetPtr=/Game/Sounds/Fort_Traps/Speed_Trap/Speed_Curve_Trap_Ready01}}
]
properties {
 attenuationSettings=&Fort_Traps_Att
 bOverrideAttenuation=true
 bOverrideConcurrency=true
 maxDistance=2000
 soundClassObject=/Game/Sounds/FortSoundClassesAndModes/Fort_Traps_Placement
 volumeMultiplier=0.5
 attenuationOverrides {bEnableListenerFocus=true, falloffDistance=2000, nonFocusVolumeAttenuation=0.5, attenuationShapeExtents {x=0}}
 concurrencyOverrides {resolutionRule=PreventNew}
}
