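# Sound cue for the jump pad trap's reset sound in Fort_Traps, reusing the jump pad placement wave.
# An Enveloper wraps the single wave player, applying a slightly raised pitch range, a rising pitch curve and a volume fade-out to reshape that sound.
# It plays at reduced volume with trap attenuation and the trap placement sound class, and its concurrency override prevents new instances at the limit.
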SoundCue /Game/Sounds/Fort_Traps/JumpPad/JumpPad_Reset_Cue
root=Enveloper_0
nodes [
 {
  id=Enveloper_0
  children [WavePlayer_107]
  properties {
   pitchMax=1.2
   pitchMin=1.1
   pitchCurve {editorCurveData {keys [{time=0}, <at 1 0.5>]}}
   volumeCurve {editorCurveData {keys [<at 0 1>,<at 1.75 0>]}}
  }
 }
 {id=WavePlayer_107, properties {soundWaveAssetPtr=&/JumpPad_Placed_001}}
]
properties {
 attenuationSettings=&Fort_Traps_Att
 bOverrideConcurrency=true
 maxDistance=6000
 soundClassObject=/Game/Sounds/FortSoundClassesAndModes/Fort_Traps_Placement
 volumeMultiplier=0.4
 concurrencyOverrides {resolutionRule=PreventNew}
}
